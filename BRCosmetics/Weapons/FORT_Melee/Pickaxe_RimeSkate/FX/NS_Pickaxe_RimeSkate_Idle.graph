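# Niagara idle effect for the RimeSkate pickaxe that draws persistent, translucent cyan glow sprites at fixed offsets on the weapon.
# A single Glow emitter spawns a burst of four particles, using spawn group to switch each one's offset, size, alpha, facing and alignment.
# The group also sets the visibility tag for three sprite renderers, merging several glow layers into one emitter.
# Alpha scales with time of day and front-end versus gameplay context.
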NiagaraSystem /BRCosmetics/Weapons/FORT_Melee/Pickaxe_RimeSkate/FX/NS_Pickaxe_RimeSkate_Idle
dependencies {usedDataInterfaces [], usedDynamicInputs [], usedModules [&SystemState]}
emitters {Glow {
  emitterUpdateScript [
   &EmitterState(
    Inactive Response=::NewEnumerator1
    Life Cycle Mode=::NewEnumerator1
    Loop Behavior=::NewEnumerator1
    Loop Duration Mode=::NewEnumerator1
    UseLoopDelay=false
   )
   &SpawnBurst_Instantaneous(Spawn Count=4, Use Spawn Probability=false, UseLoopCountLimit=false)
  ]
  particleSpawnScript [
   &FN_SetSpawnGroup()
   &InitializeParticle(
    Color=&MultiplyLinearColor(
     Linear Color=<rgba 0.1 1 1 0.25>
     Scale Factor=&MakeVector4FromVectorAndFloat(
      "Vector (RGB)"=<1 1 1>
      "Float (Alpha)"=&MergeHelper_FloatSwitch(0=0.7, 1=1, 2=4, 3=?, Index=&SpawnGroup(), Num=::NewEnumerator1)
     )
    )
    Position Offset=&MergeHelper_VectorSwitch(0=<-9 0 28.5>, 1=<-7.5891 0 40>, 2=<-15.7762 0 30>, 3=?, Index=&SpawnGroup(), Num=::NewEnumerator1)
    Sprite Size=&MergeHelper_Vector2Switch(
     0=&Multiply_Vector2D_ByFloat(Float=1, Vector2D=<37.1255 -30.3655>)
     1=<37 10>
     2=<10 28>
     3=?
     4=?
     Index=&SpawnGroup()
     Num=::NewEnumerator1
    )
    Sprite Size Mode=::NewEnumerator4
    UsePositionOffset=true
    Write Material Random=true
   )
  ]
  particleUpdateScript [
   &ParticleState(Kill Particles When Lifetime Has Elapsed=false)
   &SpriteFacingAndAlignment(
    Sprite Alignment=&MergeHelper_VectorSwitch(0=<0 0 -1>, 1=<0 -1 0>, 2=<0 0 1>, 3=?, Index=&SpawnGroup(), Num=::NewEnumerator1)
    Sprite Facing=&MergeHelper_VectorSwitch(0=<0 1 0>, 1=<0 0 1>, 2=<1 0 0>, 3=?, Index=&SpawnGroup(), Num=::NewEnumerator1)
    WriteAlignment=true
    WriteFacing=true
   )
   &DynamicMaterialParameters(Module.Param0WriteEnabled=true, Module.Param1WriteEnabled=true, Module.Param2WriteEnabled=true, Module.Param3WriteEnabled=true)
   &ScaleColor(
    Scale Alpha=&Multiply_Float(
     A=&FNTimeOfDayScalar(Night Scale=0.5, Morning Scale=0.8, Evening Scale=0.8)
     B=&FN_FrontEndFloatModifier(Gameplay=1, Front End=0.7)
    )
    ScaleA=true
    ScaleRGB=true
   )
   /SetVariables(Particles.VisibilityTag=&SpawnGroup()) {isSetParameters=true}
  ]
  rendererClasses [&NiagaraSpriteRendererProperties]
  renderers [
   {rendererClass=&NiagaraSpriteRendererProperties}
   {rendererClass=&NiagaraSpriteRendererProperties, rendererIndex=1}
   {rendererClass=&NiagaraSpriteRendererProperties, rendererIndex=2}
  ]
 }}
summary {
 systemName=NS_Pickaxe_RimeSkate_Idle
 emitters [{emitterName=Glow, rendererClasses [&NiagaraSpriteRendererProperties]}]
 userVariables [
  User.bIsFrontEnd {type=NiagaraBool}
  User.bIsFrontEndPreview {type=NiagaraBool}
  User.FrontEndScale {type=NiagaraFloat, defaultValue=0.55, description=""}
 ]
}
