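# Animation Blueprint for the RimeSkate pickaxe that adds physics-driven swing to its five-bone chain.
# The cached input pose feeds two serial Anim Dynamics stacks over chain1 to chain5, root-relative with gravity and angular limits.
# A bool blend on bIsInFrontEnd picks the front-end stack, which has lower damping, different springs, sim-space angular limits and an additive chain2 offset.
# The event graph holds no active logic.
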AnimBlueprint /BRCosmetics/Weapons/FORT_Melee/Pickaxe_RimeSkate/Meshes/RimeSkate_Axe_AnimBP
target_skeleton=&/RimeSkate_Axe_Skeleton
graphs [
 AnimGraph {nodes [
   n1 = Anim Dynamics - Bone: chain1(ComponentPose=Local To Component(LocalPose=Use cached pose 'baseCache'()), Alpha=1, ExternalForce="0,0,0") {
    angularDampingOverride=0.9999
    angularSpringConstant=60
    bAngularSpring=true
    bOverrideAngularDamping=true
    bUseGravityOverride=true
    linearDampingOverride=0.99
    linearSpringConstant=550
    simulationSpace=RootRelative
    boundBone {boneName=chain1}
    componentAppliedLinearAccClamp=<3000 3000 6000>
    componentLinearAccScale=<0.4 0.4 0.4>
    componentLinearVelScale=<0.1 0.1 0.1>
    gravityOverride {z=-980}
    physicsBodyDefinitions [{
      boundBone {boneName=chain1}
      boxExtents=<10 10 10>
      constraintSetup {angularLimitsMax=<0 0 55>, angularLimitsMin=<0 0 -55>, angularTarget=<1 0 0>}
      localJointOffset=<1 0 0>
     }]
   }
   n2 = Anim Dynamics - Bone: chain2(ComponentPose=@n1, Alpha=1, ExternalForce="0,0,0") {
    angularDampingOverride=0.9999
    angularSpringConstant=550
    bOverrideAngularDamping=true
    bUseGravityOverride=true
    linearDampingOverride=0.99
    linearSpringConstant=550
    simulationSpace=RootRelative
    worldSpaceMinimumScale=0.1
    boundBone {boneName=chain2}
    componentAppliedLinearAccClamp=<3000 3000 6000>
    componentLinearAccScale=<0.3 0.3 1>
    componentLinearVelScale=<0.2 0.2 0.1>
    gravityOverride {z=-980}
    physicsBodyDefinitions [{
      boundBone {boneName=chain2}
      boxExtents=<10 10 10>
      constraintSetup {angularLimitsMax=<5 40 20>, angularLimitsMin=<-5 -80 -20>, angularTarget=<1 0 0>}
      localJointOffset=<1 0.1 0.1>
     }]
   }
   n3 = Anim Dynamics - Bone: chain3(ComponentPose=@n2, Alpha=1, ExternalForce="0,0,0") {
    angularDampingOverride=0.9999
    angularSpringConstant=550
    bOverrideAngularDamping=true
    bUseGravityOverride=true
    linearDampingOverride=0.99
    linearSpringConstant=550
    simulationSpace=RootRelative
    worldSpaceMinimumScale=0.1
    boundBone {boneName=chain3}
    componentAppliedLinearAccClamp=<3000 3000 6000>
    componentLinearAccScale=<0.3 0.3 1>
    componentLinearVelScale=<0.2 0.2 0.1>
    gravityOverride {z=-980}
    physicsBodyDefinitions [{
      boundBone {boneName=chain3}
      boxExtents=<10 10 10>
      constraintSetup {angularLimitsMax=<10 20 30>, angularLimitsMin=<-10 -20 -30>, angularTarget=<1 0 0>}
      localJointOffset=<1 0.1 0.1>
     }]
   }
   n4 = Anim Dynamics - Bone: chain4(ComponentPose=@n3, Alpha=1, ExternalForce="0,0,0") {
    angularDampingOverride=0.9999
    angularSpringConstant=550
    bOverrideAngularDamping=true
    bUseGravityOverride=true
    linearDampingOverride=0.99
    linearSpringConstant=550
    simulationSpace=RootRelative
    worldSpaceMinimumScale=0.1
    boundBone {boneName=chain4}
    componentAppliedLinearAccClamp=<3000 3000 6000>
    componentLinearAccScale=<0.3 0.3 1>
    componentLinearVelScale=<0.2 0.2 0.1>
    gravityOverride {z=-980}
    physicsBodyDefinitions [{
      boundBone {boneName=chain4}
      boxExtents=<10 10 10>
      constraintSetup {angularLimitsMax=<10 20 30>, angularLimitsMin=<-10 -20 -30>, angularTarget=<1 0 0>}
      localJointOffset=<1 0.1 0.1>
     }]
   }
   n5 = Anim Dynamics - Bone: chain5(ComponentPose=@n4, Alpha=1, ExternalForce="0,0,0") {
    angularDampingOverride=0.9999
    angularSpringConstant=550
    bOverrideAngularDamping=true
    bUseGravityOverride=true
    linearDampingOverride=0.99
    linearSpringConstant=550
    simulationSpace=RootRelative
    worldSpaceMinimumScale=0.1
    boundBone {boneName=chain5}
    componentAppliedLinearAccClamp=<3000 3000 6000>
    componentLinearAccScale=<0.3 0.3 1>
    componentLinearVelScale=<0.2 0.2 0.1>
    gravityOverride {z=-980}
    physicsBodyDefinitions [{
      boundBone {boneName=chain5}
      boxExtents=<10 10 10>
      constraintSetup {angularLimitsMax=<10 20 30>, angularLimitsMin=<-10 -20 -30>, angularTarget=<1 0 0>}
      localJointOffset=<1 0.1 0.1>
     }]
   }
   n6 = Anim Dynamics - Bone: chain1(ComponentPose=Local To Component(LocalPose=Use cached pose 'baseCache'()), Alpha=1, ExternalForce="0,0,0") {
    angularDampingOverride=0.993
    bOverrideAngularDamping=true
    bUseGravityOverride=true
    linearDampingOverride=0.99
    linearSpringConstant=550
    simulationSpace=RootRelative
    boundBone {boneName=chain1}
    componentAppliedLinearAccClamp=<3000 3000 6000>
    componentLinearAccScale=<0.25 0.25 0.25>
    componentLinearVelScale=<0.1 0.1 0.1>
    gravityOverride {z=-980}
    physicsBodyDefinitions [{
      boundBone {boneName=chain1}
      boxExtents=<9.5 9.5 9.5>
      constraintSetup {angularLimitsMax=<0 0 55>, angularLimitsMin=<0 0 -55>, angularTarget=<1 0 0>}
      localJointOffset=<1 0 0>
     }]
    simSpaceSettings {maxAngularAcceleration=15, maxAngularVelocity=8, simSpaceAngularAlpha=0.3}
   }
   n7 = "Transform (Modify) Bone - Bone: chain2"(ComponentPose=@n6, Alpha=1, Translation="-1,0,0.2", Rotation="0,0,0", Scale="1,1,1") {translationMode=BMM_Additive, translationSpace=BCS_BoneSpace, boneToModify {boneName=chain2}, translation {x=-1, z=0.2}}
   n8 = Anim Dynamics - Bone: chain2(ComponentPose=@n7, Alpha=1, ExternalForce="0,0,0") {
    angularDampingOverride=0.993
    bAngularSpring=true
    bOverrideAngularDamping=true
    bUseGravityOverride=true
    linearDampingOverride=0.99
    linearSpringConstant=550
    simulationSpace=RootRelative
    boundBone {boneName=chain2}
    componentAppliedLinearAccClamp=<3000 3000 6000>
    componentLinearAccScale=<0.25 0.25 0.25>
    componentLinearVelScale=<0.2 0.2 0.1>
    gravityOverride {z=-980}
    physicsBodyDefinitions [{
      boundBone {boneName=chain2}
      boxExtents=<9.5 9.5 9.5>
      constraintSetup {angularLimitsMax=<5 40 20>, angularLimitsMin=<-5 -90 -20>, angularTarget=<1 0 0>}
      localJointOffset=<1 0.1 0.1>
     }]
    simSpaceSettings {maxAngularAcceleration=15, maxAngularVelocity=8, simSpaceAngularAlpha=0.3}
   }
   n9 = Anim Dynamics - Bone: chain3(ComponentPose=@n8, Alpha=1, ExternalForce="0,0,0") {
    angularDampingOverride=0.993
    bAngularSpring=true
    bOverrideAngularDamping=true
    bUseGravityOverride=true
    linearDampingOverride=0.99
    linearSpringConstant=550
    simulationSpace=RootRelative
    boundBone {boneName=chain3}
    componentAppliedLinearAccClamp=<3000 3000 6000>
    componentLinearAccScale=<0.25 0.25 0.25>
    componentLinearVelScale=<0.2 0.2 0.1>
    gravityOverride {z=-980}
    physicsBodyDefinitions [{
      boundBone {boneName=chain3}
      boxExtents=<9.5 9.5 9.5>
      constraintSetup {angularLimitsMax=<10 20 30>, angularLimitsMin=<-10 -20 -30>, angularTarget=<1 0 0>}
      localJointOffset=<1 0.1 0.1>
     }]
    simSpaceSettings {maxAngularAcceleration=15, maxAngularVelocity=8, simSpaceAngularAlpha=0.3}
   }
   n10 = Anim Dynamics - Bone: chain4(ComponentPose=@n9, Alpha=1, ExternalForce="0,0,0") {
    angularDampingOverride=0.993
    bAngularSpring=true
    bOverrideAngularDamping=true
    bUseGravityOverride=true
    linearDampingOverride=0.99
    linearSpringConstant=550
    simulationSpace=RootRelative
    boundBone {boneName=chain4}
    componentAppliedLinearAccClamp=<3000 3000 6000>
    componentLinearAccScale=<0.25 0.25 0.25>
    componentLinearVelScale=<0.2 0.2 0.1>
    gravityOverride {z=-980}
    physicsBodyDefinitions [{
      boundBone {boneName=chain4}
      boxExtents=<9.5 9.5 9.5>
      constraintSetup {angularLimitsMax=<10 20 30>, angularLimitsMin=<-10 -20 -30>, angularTarget=<1 0 0>}
      localJointOffset=<1 0.1 0.1>
     }]
    simSpaceSettings {maxAngularAcceleration=15, maxAngularVelocity=8, simSpaceAngularAlpha=0.3}
   }
   n11 = Anim Dynamics - Bone: chain5(ComponentPose=@n10, Alpha=1, ExternalForce="0,0,0") {
    angularDampingOverride=0.993
    bAngularSpring=true
    bOverrideAngularDamping=true
    bUseGravityOverride=true
    linearDampingOverride=0.99
    linearSpringConstant=550
    simulationSpace=RootRelative
    boundBone {boneName=chain5}
    componentAppliedLinearAccClamp=<3000 3000 6000>
    componentLinearAccScale=<0.25 0.25 0.25>
    componentLinearVelScale=<0.2 0.2 0.1>
    gravityOverride {z=-980}
    physicsBodyDefinitions [{
      boundBone {boneName=chain5}
      boxExtents=<9.5 9.5 9.5>
      constraintSetup {angularLimitsMax=<10 20 30>, angularLimitsMin=<-10 -20 -30>, angularTarget=<1 0 0>}
      localJointOffset=<1 0.1 0.1>
     }]
    simSpaceSettings {maxAngularAcceleration=15, maxAngularVelocity=8, simSpaceAngularAlpha=0.3}
   }
   Output Pose(Result=Blend Poses by bool(BlendPose_0=Component To Local(ComponentPose=@n11), BlendPose_1=Component To Local(ComponentPose=@n5), bActiveValue=Get bIsInFrontEnd()) {blendType=Linear}) {layerGroup=None}
   AnimGraphNode_SubInput_0 = Input Pose()
   Save cached pose 'baseCache'(Pose=@AnimGraphNode_SubInput_0) {cacheName=baseCache}
  ]}
 EventGraph {nodes [
   TryGetPawnOwner()
   Event BlueprintUpdateAnimation()
  ]}
]
variables []
